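# Material function for the holographic RL232SE wheel: a black, metallic surface with animated emissive glow, layered over MF_WheelParent_Simple.
# Emissive multiplies mask-blended rim colors by summed Fresnel, sine-pulsed AO and a rotating second-UV radial gradient broken up by a pixel mask.
# Rim and tire normal maps are blended and also feed the Fresnel.
# An Inverted static switch provides a white version by flipping AO masks and altering intensity.
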MaterialFunction /VehicleCosmetics/Wheels/RL232SE/Materials/MF_Wheel_RL232SE_Holographic
graph {nodes [
  n1 = SetMaterialAttributes!(
   MaterialAttributes=MaterialFunctionCall() {materialFunction=&MF_WheelParent_Simple}.MaterialAttributes
   EmissiveColor=NamedRerouteUsage()
   BaseColor=Constant()
   Metallic=Constant() {r=1}
   Normal=NamedRerouteUsage()
   Specular
   Roughness=Constant()
  ) {attributeSetTypes [
    #1
    #2
    #3
    #4
    #5
    #6
   ]}
  FunctionOutput(None=@n1)
  n2 = Constant() {r=1}
  n3 = TextureSampleParameter2D() {
   group=1_Input Textures
   parameterName=AO_mask
   samplerType=SAMPLERTYPE_Masks
   texture=/VehicleCosmetics/Wheels/RL232SE/Textures/T_Tire_RL232SE_AO_M
  }
  n4 = Add(A=@n3.R, B=ScalarParameter() {defaultValue=0.175, parameterName=AmbientOffset}) {constB=0.1}
  n5 = ScalarParameter() {defaultValue=20, parameterName=InvertedAmbientPower}
  n6 = LinearInterpolate(A=@n3.G, B=Clamp(None=Power(Base=@n4, Exp=@n5)), Alpha=NamedRerouteUsage()) {desc=ao}
  n7 = VectorParameter() {parameterName=Rim_ColorA, defaultValue=<rgba 1 0.1779 0.8796 1>}
  n8 = Multiply(A=Multiply(A=NamedRerouteUsage()) {constB=2, desc=glowy triangles}, B=@n7.RGB)
  n9 = TextureSampleParameter2D() {parameterName=Rim_MaskRGBA, samplerType=SAMPLERTYPE_Masks, texture=/VehicleCosmetics/Wheels/RL232SE/Textures/T_Rim_RL232SE_M}
  n10 = LinearInterpolate(
   A=VectorParameter() {parameterName=Rim_ColorA, defaultValue=<rgba 1 0.1779 0.8796 1>}.RGB
   B=VectorParameter() {parameterName=Rim_ColorC, defaultValue=<rgba 0.8879 0.2346 0.0561 1>}.RGB
   Alpha=Add(A=@n9.R, B=@n9.R)
  )
  n11 = VectorParameter() {parameterName=Rim_ColorB, defaultValue=<rgba 0 1 0 1>}
  n12 = TextureSampleParameter2D() {
   group=1_Input Textures
   parameterName=Rim_normal
   samplerType=SAMPLERTYPE_Normal
   texture=/VehicleCosmetics/Wheels/RL232SE/Textures/T_Rim_RL232SE_N
  }
  n13 = TextureSampleParameter2D() {
   group=1_Input Textures
   parameterName=Tire_normal
   samplerType=SAMPLERTYPE_Normal
   texture=/VehicleCosmetics/Wheels/RL232SE/Textures/T_Tire_RL232SE_N
  }
  n14 = LinearInterpolate(A=@n12.RGB, B=@n13.RGB, Alpha=NamedRerouteUsage()) {desc=normal}
  n15 = Fresnel(Normal=Transform(None=NamedRerouteDeclaration(None=@n14) {name=Normal_out, nodeColor=<rgba 0 1 0.8594 1>})) {exponent=3}
  n16 = Constant() {r=0.8}
  n17 = Constant() {r=2}
  n18 = Multiply(A=ScalarParameter() {defaultValue=0.5, parameterName=FresnelStrength}, B=Add(A=Power(Base=@n15, Exp=@n16), B=Power(Base=@n15, Exp=@n17)))
  n19 = Constant() {r=3}
  n20 = Multiply(A=TextureSample(UVs=Rotator() {constCoordinate=1, desc=rotate on 2nd uv channel, speed=-0.33}) {texture=/VehicleCosmetics/Wheels/RL232SE/Textures/T_Radialgrad}.R) {constB=2}
  n21 = TextureSampleParameter2D() {
   group=1_Input Textures
   parameterName=Pixel-mask
   samplerType=SAMPLERTYPE_Masks
   texture=/VehicleCosmetics/Wheels/RL232SE/Textures/T_Rim_RL232SE_pixelmask
  }
  n22 = Add(A=@n21.G, B=NamedRerouteUsage())
  n23 = Multiply(A=Multiply(A=Power(Base=@n20), B=Frac(None=@n22)), B=@n21.G)
  n24 = LinearInterpolate(
   A=StaticSwitchParameter(True=Multiply(A=@n19, B=@n23) {constA=3}, False=@n23) {desc="Inverted (white version) switch", parameterName=Inverted}
   Alpha=NamedRerouteUsage()
  ) {constB=0, desc=mask black for tire}
  n25 = Add(A=StaticSwitchParameter(True=OneMinus(None=@n3.R), False=@n3.R) {desc="Inverted (white version) switch", parameterName=Inverted}) {constB=0.15}
  n26 = LinearInterpolate(
   A=StaticSwitchParameter(True=OneMinus(None=@n3.G), False=@n3.G) {desc="Inverted (white version) switch", parameterName=Inverted}
   B=Power(Base=Clamp(None=@n25)) {constExponent=4}
   Alpha=NamedRerouteUsage()
  )
  n27 = Multiply(A=ScalarParameter() {defaultValue=0.4, parameterName=Rim_Pulsespeed}, B=Time())
  n28 = Sine(None=@n27)
  n29 = Divide(B=Add(B=@n28) {constA=3}) {constA=2}
  n30 = OneMinus(None=Clamp(None=Power(Base=@n26, Exp=@n29)))
  n31 = Add(A=@n28)
  n32 = Multiply(A=@n30, B=Clamp(None=Divide(A=@n31) {constB=10}))
  n33 = Multiply(
   A=LinearInterpolate(A=LinearInterpolate(A=@n10, B=@n11.RGB, Alpha=@n9.G), B=Add(A=@n8, B=@n7.RGB), Alpha=NamedRerouteUsage()) {desc=Color}
   B=Add(A=Add(A=@n18, B=@n24), B=Add(A=@n30, B=@n32))
  )
  n34 = ScalarParameter() {defaultValue=2, parameterName=EmissiveStrength}
  NamedRerouteDeclaration(None=Multiply(A=StaticSwitchParameter(True=@n6, False=@n2) {desc="Inverted (white version) switch", parameterName=Inverted}, B=Multiply(A=@n33, B=@n34))) {name=Emissive_Out, nodeColor=<rgba 1 0 0.8203 1>}
  NamedRerouteDeclaration(None=@n21.R) {name=GlowyriangleMask, nodeColor=<rgba 1 0.7734 0 1>}
  NamedRerouteDeclaration(None=@n27) {name=PulseTimer, nodeColor=<rgba 0.4531 0 1 1>}
  NamedRerouteDeclaration(None=VertexColor().R) {name=VC_R, nodeColor=<rgba 1 0 0 1>}
  OneMinus()
  OneMinus(None=@n21.G)
  Power()
  VectorParameter() {parameterName=Rim_ColorBase, defaultValue=<rgba 0 0 1 1>}
 ]}
